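# Material function that builds an animated ghostly emissive from swirling two-color noise, a colored Fresnel rim and a sine-driven color pulse.
# A swirl texture scrolls over time with a warp offset added to its UVs, blending two tint colors.
# The Fresnel normal is an alpha lerp, with the input normal and a panning detail normal combined through BlendAngleCorrectedNormals.
# Parameters in the Ghost Noise group expose colors, speeds and strengths.
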MaterialFunction /Game/Global/Materials/MaterialGym/Materials/GhostNoise/MaterialFunctions/MF_GhostNoiseEmissive
graph {nodes [
  FunctionOutput(None=NamedRerouteUsage())
  n1 = FunctionInput() {bUsePreviewValueAsDefault=true, inputName=NormalAlpha, inputType=FunctionInput_Scalar, sortPriority=1, previewValue=<1 0 0 1>}
  NamedRerouteDeclaration(None=@n1) {name=Alpha, nodeColor=<rgba 1 0.7969 0 1>}
  n2 = LinearInterpolate(A=NamedRerouteUsage(), B=NamedRerouteUsage(), Alpha=NamedRerouteUsage())
  n3 = Saturate(None=Fresnel(
    ExponentIn=ScalarParameter() {defaultValue=1.3113, group=Ghost Noise, parameterName=GhostFresnel_Exponent, sortPriority=11}
    BaseReflectFractionIn=Constant()
    Normal=Transform(None=@n2)
   ) {baseReflectFraction=0})
  n4 = VectorParameter() {group=Ghost Noise, parameterName=Ghost_FresnelColor, sortPriority=10, defaultValue=<rgba 0.1258 0.0113 0.14 0>}
  n5 = Add(A=@n4.RGB, B=NamedRerouteUsage())
  n6 = Add(
   A=NamedRerouteUsage()
   B=Multiply(A=Multiply(A=@n3, B=@n5), B=ScalarParameter() {defaultValue=1.2471, group=Ghost Noise, parameterName=GhostFresnel_Strength, sortPriority=12})
  )
  NamedRerouteDeclaration(None=Add(A=NamedRerouteDeclaration(None=@n6) {name=GhostFresnelEmissive, nodeColor=<rgba 0.5703 0 1 1>}, B=NamedRerouteUsage())) {name=GhostEmissive, nodeColor=<rgba 0.125 1 0 1>}
  n7 = TextureSample() {
   samplerType=SAMPLERTYPE_Normal
   texture=/Game/Characters/Player/Female/Medium/Bodies/F_MED_RebelClaw_Aviator/Skins/Cobalt/Textures/T_Blank_Normal
  }
  NamedRerouteDeclaration(None=FunctionInput(Preview=@n7.RGB) {bUsePreviewValueAsDefault=true, inputName=Normal}) {name=InNormal, nodeColor=<rgba 0 0.7187 1 1>}
  n8 = NamedRerouteUsage()
  n9 = TextureSample(UVs=Panner(Coordinate=Multiply(A=@n8) {constB=20}) {speedY=0.2}) {
   samplerType=SAMPLERTYPE_Normal
   texture=/Game/Characters/Player/Male/Medium/Bodies/M_MED_Cat_Burglar_Ghost/Textures/T_M_MED_Cat_Burglar_Ghost_N1
  }
  NamedRerouteDeclaration(None=MaterialFunctionCall(BaseNormal=NamedRerouteUsage(), AdditionalNormal=@n9.RGB) {materialFunction=&BlendAngleCorrectedNormals}.Result) {name=GhostNormal, nodeColor=<rgba 0 0.3437 1 1>}
  n10 = VectorParameter() {group=Ghost Noise, parameterName=GhostNoise_Color_01, sortPriority=24, defaultValue=<rgba 0.0363 0.0045 0.0521 1>}
  n11 = VectorParameter() {group=Ghost Noise, parameterName=GhostNoise_Color_02, sortPriority=25, defaultValue=<rgba 0.5592 0.0683 1 1>}
  n12 = VectorParameter() {group=Ghost Noise, parameterName=GhostNoise_Speed, sortPriority=22, defaultValue=<rgba 0 0 -0.02 1>}
  n13 = Multiply(A=Time(), B=@n12.RGB)
  n14 = VectorParameter() {group=Ghost Noise, parameterName=GhostNoise_Tile, sortPriority=21, defaultValue=<rgba 1 1 1 1>}
  n15 = ScalarParameter() {defaultValue=0.05, group=Ghost Noise, parameterName=GhostNoise_WarpAmount, sortPriority=23}
  n16 = NamedRerouteUsage()
  n17 = ComponentMask(None=Add(A=Multiply(A=@n13, B=@n14.RGB), B=Multiply(A=@n15, B=@n16))) {g=true, r=true}
  n18 = NamedRerouteDeclaration(None=FunctionInput(Preview=TextureCoordinate()) {bUsePreviewValueAsDefault=true, inputName=UVInput, inputType=FunctionInput_Vector2}) {name=UVInput, nodeColor=<rgba 0 1 0.6016 1>}
  n19 = ComponentMask(None=TextureSample(UVs=Add(A=@n17, B=@n18)) {samplerType=SAMPLERTYPE_LinearColor, texture=/Game/Global/Materials/MaterialGym/Materials/GhostNoise/Textures/T_ghost_swirls_01}.RGB) {r=true}
  NamedRerouteDeclaration(None=Multiply(
    A=LinearInterpolate(A=@n10.RGB, B=@n11.RGB, Alpha=@n19)
    B=ScalarParameter() {defaultValue=1, group=Ghost Noise, parameterName=GhostNoise_EmissiveStrength, sortPriority=26}
   )) {name=GhostNoise, nodeColor=<rgba 0.0547 1 0 1>}
  n20 = VectorParameter() {group=Ghost Noise, parameterName=Ghost_PulseColor, sortPriority=30, defaultValue=<rgba 0.1258 0.0113 0.14 0>}
  n21 = Time()
  n22 = ScalarParameter() {defaultValue=0.25, group=Ghost Noise, parameterName=Ghost_PulseSpeed, sortPriority=31}
  n23 = ConstantBiasScale(None=Sine(None=Multiply(A=@n21, B=@n22)))
  n24 = ScalarParameter() {defaultValue=1, group=Ghost Noise, parameterName=Ghost_PulseBrightness}
  NamedRerouteDeclaration(None=Multiply(A=@n20.RGB, B=Multiply(A=@n23, B=@n24))) {name=Pulse, nodeColor=<rgba 1 0 0.4687 1>}
 ]}
