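# Default Lit example material in the MaterialGym that demonstrates the MF_GhostNoiseEmissive material function.
# A Fresnel term built on the world-space vertex normal, with scalar parameters for exponent and base reflect fraction, feeds the function's NormalAlpha input.
# The function's Normal and UVInput inputs are left unconnected.
# Usage is enabled for skeletal meshes and mesh deformers, with automatic usage setting in the editor turned off.
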Material /Game/Global/Materials/MaterialGym/Materials/GhostNoise/MaterialFunctions/M_GhostNoiseExample
graph {
 properties {bAutomaticallySetUsageInEditor=false, bUsedWithMeshDeformer=true, bUsedWithSkeletalMesh=true, usedShadingModels=Default Lit}
 nodes [
  Constant()
  MaterialFunctionCall!(
   Normal
   UVInput
   NormalAlpha=Fresnel(
    ExponentIn=ScalarParameter() {defaultValue=1.5, parameterName=ExponentIn}
    BaseReflectFractionIn=ScalarParameter() {defaultValue=0.1, parameterName=BaseReflectFractionIn}
    Normal=VertexNormalWS()
   )
  ) {materialFunction=&/MF_GhostNoiseEmissive}
 ]
}
